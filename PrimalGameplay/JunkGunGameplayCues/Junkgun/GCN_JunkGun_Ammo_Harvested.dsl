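; Burst gameplay cue for the Junk Gun that plays suction effects and audio when ammo is harvested.
; It feeds the impact location and normalized magnitude into the first particle component, and attaches a pickup sound to the instigator weapon, switching to a full-clip sound when magazine count plus one equals clip size.
; A looping timer keeps the particle's weapon location on the muzzle until a delayed cleanup clears it.
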(event EventOnBurst(EitherVFX-preferthisversion)
  (params (MyTarget :type Actor%20Object%20Reference) (Parameters :type Gameplay%20Cue%20Parameters%20Structure%20%28by%20ref%29) (ParticleComponents :type Array%20of%20FXSystem%20Component%20Object%20References) (AudioComponents :type Array%20of%20Audio%20Component%20Object%20References) (BurstCameraShakeInstance :type Camera%20Shake%20Base%20Object%20Reference) (BurstDecalInstance :type Decal%20Actor%20Object%20Reference))
  (bind (_normalizedmagnitude _rawmagnitude _effectcontext _matchedtagname _originaltag _aggregatedsourcetags _aggregatedtargettags _location _normal _instigator) (Ability|GameplayCue|BreakGameplayCueParameters Parameters))
  (bind _output (Utilities|Array|Get(acopy) ParticleComponents))
  (bind _asfort_weapon (Utilities|Casting|CastToFortWeapon _instigator))
  (Variables|Default|SetWeaponRef _asfort_weapon)
  (bind _returnvalue (Utilities|Time|SetTimerbyEvent (delegate Custom|UpdateMuzzleLocation) (Variables|Default|GetUpdateTimer) true))
  (Variables|Default|SetUpdateTimer_TH _returnvalue)
  (Effects|Components|ParticleSystem|SetVectorParameter _output "ImpactLocation" _location)
  (Effects|Components|ParticleSystem|SetFloatParameter _output "CollectingMaterial" _normalizedmagnitude)
  (Effects|Components|ParticleSystem|SetVectorParameter _output "WeaponLocation" (Effects|GetMuzzleLocation (Variables|Default|GetWeaponRef) -1))
  (Audio|SpawnSoundAttached :Sound (select (== (+ (FortWeapon|GetMagazineAmmoCount (Variables|Default|GetWeaponRef)) 1) (FortWeapon|GetBulletsPerClip (Variables|Default|GetWeaponRef))) "/PrimalGameplay/Items/JunkGun/Assets/Audio/JunkGun_Launcher_Suction_Ammo_Full_Cue.JunkGun_Launcher_Suction_Ammo_Full_Cue" "/PrimalGameplay/Items/JunkGun/Assets/Audio/JunkGun_Launcher_Suction_Ammo_Pickup_Cue.JunkGun_Launcher_Suction_Ammo_Pickup_Cue") :AttachToComponent (Variables|/Script/Engine.Actor|GetRootComponent _instigator) :AttachPointName "None" :LocationType "SnapToTargetIncludingScale" :bStopWhenAttachedToDestroyed false :VolumeMultiplier 1.0 :PitchMultiplier 1.0 :StartTime 0.0 :bAutoDestroy true)
  (Utilities|Time|SetTimerbyEvent (EventDispatchers|CreateEvent :!event Custom|CleanUp) (Variables|Default|GetTimeBeforeCleanUp)))

(event Custom|UpdateMuzzleLocation
  (Effects|Components|ParticleSystem|SetVectorParameter (Utilities|Array|Get(acopy) (stale EventOnBurst(EitherVFX-preferthisversion) ParticleComponents)) "WeaponLocation" (Effects|GetMuzzleLocation (Variables|Default|GetWeaponRef) -1)))

(event Custom|CleanUp
  (Utilities|Time|ClearandInvalidateTimerbyHandle 0 (Variables|Default|GetUpdateTimer_TH)))

(fn ConstructionScript)
