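; Looping gameplay cue for Junk Gun suction, driving grinder, suction and impact FX plus layered audio.
; On start it caches pawn, weapon and FX, attaches audio, listens for ammo pickups and times sound indicators for remote players.
; Tick feeds weapon hit, resource, ammo and pickup-timed volume curve values into FX/audio, deriving underwater state from aim pitch and immersion to pass water-surface data.
; Removal fades audio and clears timer and listener.
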(event Custom|ShowSoundIndicator
  (bind _pawnreference (Variables|Default|GetPawnReference))
  (if (not (Pawn|IsLocallyControlled _pawnreference))
    (Fortnite|ShowSoundIndicatorLocal :WorldContextObject _pawnreference :Location (Transformation|GetActorLocation _pawnreference) :MaxAudibleDistance 5e+03 :IndicatorType "Gunshot" :Instigator self :Tint "(R=0.000000,G=0.000000,B=0.000000,A=0.000000)" :AttenuationMaxRangeMultiplier 1.0 :AttenuationIntensityMultiplier 1.0)))

(event EventOnLoopingStart(EitherVFX-preferthisversion)
  (params (MyTarget :type Actor%20Object%20Reference) (Parameters :type Gameplay%20Cue%20Parameters%20Structure%20%28by%20ref%29) (ParticleComponents :type Array%20of%20FXSystem%20Component%20Object%20References) (AudioComponents :type Array%20of%20Audio%20Component%20Object%20References))
  (bind _asfort_player_pawn (Utilities|Casting|CastToFortPlayerPawn MyTarget))
  (bind _output_get (Variables|Default|SetPawnReference _asfort_player_pawn))
  (bind _currentweapon (Variables|/Script/FortniteGame.FortPawn|GetCurrentWeapon _output_get)
    (:then
      (bind _output_get_2 (Variables|Default|SetCurrentWeapon _currentweapon))
      (bind _asynctaskproxy (GameplayEventRouter|StartListeningToEvent 0 (GameplayEventRouter|GetGlobalEventRouter _output_get_2) "/PrimalGameplay/Items/JunkGun/GameplayEvent_OnJunkGunPickupAmmo.GameplayEvent_OnJunkGunPickupAmmo" _output_get_2)
        (:then
          (Variables|Default|SetAsync_ListenToAmmoPickup _asynctaskproxy)
          (bind _result (ArrayMacros|SafeArrayGet ParticleComponents)
            (:%20%20
              (Variables|Default|SetLooping%20Grinder%20FX _result)
              (bind _result_2 (ArrayMacros|SafeArrayGet ParticleComponents 1)
                (:%20%20
                  (bind _output_get_3 (Variables|Default|SetLooping%20Suction%20FX _result_2))
                  (bind _result_3 (ArrayMacros|SafeArrayGet ParticleComponents 2)
                    (:%20%20
                      (Variables|Default|SetLooping%20Impact%20FX _result_3)
                      (Effects|Components|ParticleSystem|SetBoolParameter _output_get_3 "bIsHitting" false)
                      (Actor|Tick|SetActorTickEnabled true)
                      (Transformation|AttachComponentToComponent (Variables|Default|GetFortLayeredAudio) (Variables|/Script/Engine.Actor|GetRootComponent (Variables|Default|GetCurrentWeapon)) "None" "SnapToTarget" "SnapToTarget" "SnapToTarget" true)
                      (bind _returnvalue (Utilities|Time|SetTimerbyEvent (EventDispatchers|CreateEvent :!event Custom|ShowSoundIndicator) (Variables|Default|GetSoundIndicatorTimer) true))
                      (Variables|Default|SetSoundIndicator_TH _returnvalue))))))))
        (:OnEventReceived
          (CallFunction|OnAmmoPickup))) ; Event type: GameplayEvent_OnJunkGunPickupAmmo (can't use a variable for that pin)
    )
    (:else)))

(event EventTick
  (params (DeltaSeconds :type Float%20%28single-precision%29))
  (bind _fortlayeredaudio (Variables|Default|GetFortLayeredAudio))
  (bind _looping_suction_fx (Variables|Default|GetLooping%20Suction%20FX))
  (bind _looping_suction_fx_2 (Variables|Default|GetLooping%20Suction%20FX))
  (bind _looping_impact_fx (Variables|Default|GetLooping%20Impact%20FX))
  (Utilities|IsValid (Variables|Default|GetLooping%20Suction%20FX)
    (:Is%20Valid
      (bind _currentweapon (Variables|Default|GetCurrentWeapon)
        (:then
          (bind (_bishit _hitlocation _bisnotfull _storedresource _normal) (Class|JunkGunInterface|GetJunkGunSuctionValues(Message) _currentweapon))
          (Effects|Components|ParticleSystem|SetBoolParameter _looping_suction_fx_2 "bIsHitting" _bishit)
          (Effects|Components|ParticleSystem|SetVectorParameter _looping_suction_fx_2 "ImpactLocation" _hitlocation)
          (Effects|Components|ParticleSystem|SetBoolParameter _looping_suction_fx_2 "bIsCollecting" _bisnotfull)
          (Effects|Components|ParticleSystem|SetFloatParameter _looping_suction_fx_2 "CollectingMaterial" _storedresource)
          (Effects|Components|ParticleSystem|SetVectorParameter _looping_impact_fx "ImpactLocation" _hitlocation)
          (Effects|Components|ParticleSystem|SetVectorParameter _looping_impact_fx "Impact_Orientation" _normal)
          (Effects|Components|ParticleSystem|SetFloatParameter _looping_impact_fx "CollectingMaterial" _storedresource)
          (Class|FortLayeredAudioComponent|SetFloatParameterExt _fortlayeredaudio "FullAmmo" (/ (Math|Conversions|ToFloat(Integer) (FortWeapon|GetMagazineAmmoCount (Variables|Default|GetCurrentWeapon))) (Math|Float|SelectFloat 1.0 (Math|Conversions|ToFloat(Integer) (FortWeapon|GetBulletsPerClip (Variables|Default|GetCurrentWeapon))) (== (FortWeapon|GetBulletsPerClip (Variables|Default|GetCurrentWeapon)) 0))))
          (Class|FortLayeredAudioComponent|SetFloatParameterExt _fortlayeredaudio "Hitting" (Math|Conversions|ToFloat(Boolean) _bisnotfull))
          (bind _output_get (Variables|Default|SetCurrentDepthValue (Math|Conversions|ToFloat(Boolean) (and (< (Math|Vector|DotProduct (Math|Vector|GetForwardVector (Pawn|GetBaseAimRotation (Variables|Default|GetPawnReference))) "0, 0,1.000000") 0.1) (> (WaterBody|GetWaterImmersionDepth (Variables|Default|GetPawnReference)) (Math|Float|MapRangeClamped (Math|Vector|DotProduct (Math|Vector|GetForwardVector (Pawn|GetBaseAimRotation (Variables|Default|GetPawnReference))) "0, 0,1.000000") -1.0 0.0 (Variables|Default|GetDepthMin) (Variables|Default|GetDepthMax)))))))
          (Class|FortLayeredAudioComponent|SetFloatParameterExt _fortlayeredaudio "Underwater" _output_get)
          (Effects|Components|ParticleSystem|SetFloatParameter (Variables|Default|GetLooping%20Suction%20FX) "Underwater" _output_get)
          (Audio|Parameter|SetFloatParameter (Variables|Default|GetFortLayeredAudio) "AmmoPickup" (Math|Float|SelectFloat (Math|Curves|GetFloatValue (Variables|Default|GetVolumeCurve) (- (Utilities|Time|GetGameTimeinSeconds (Variables|Default|GetCurrentWeapon)) (Variables|Default|GetLastAmmoPickupTime))) 1.0 (Math|Float|InRange(Float) (Math|Float|MapRangeClamped (- (Utilities|Time|GetGameTimeinSeconds (Variables|Default|GetCurrentWeapon)) (Variables|Default|GetLastAmmoPickupTime)) 0.0 3.0 0.0 1.0) 0.0 1.0 false false)) :!class /Script/AudioExtensions.AudioParameterControllerInterface)
          (if (> (Variables|Default|GetCurrentDepthValue) 0.0)
            (bind _outwatersurfaceinfo (WaterBody|GetWaterSurfaceInfoStruct (WaterBody|GetCurrentWaterBody (Variables|Default|GetPawnReference)) (Transformation|GetActorLocation (Variables|Default|GetCurrentWeapon)) true false false))
            (Effects|Components|ParticleSystem|SetVectorParameter _looping_suction_fx "WaterPlaneLocation" (Utilities|Struct|BreakWaterSurfaceInfo _outwatersurfaceinfo))
            (Effects|Components|ParticleSystem|SetIntParameter _looping_suction_fx "WaterBodyIndex" (.WaterBodyIdx (Utilities|Struct|BreakWaterSurfaceInfo _outwatersurfaceinfo)))
            (Effects|Components|ParticleSystem|SetFloatParameter _looping_suction_fx "WaterDepth" (.WaterDepth (Utilities|Struct|BreakWaterSurfaceInfo _outwatersurfaceinfo)))
            (Effects|Components|ParticleSystem|SetVectorParameter _looping_suction_fx "WaterPlaneNormal" (.WaterPlaneNormal (Utilities|Struct|BreakWaterSurfaceInfo _outwatersurfaceinfo)))))
        (:else)))))

(event EventOnRemoval(EitherVFX-preferthisversion)
  (params (MyTarget :type Actor%20Object%20Reference) (Parameters :type Gameplay%20Cue%20Parameters%20Structure%20%28by%20ref%29) (ParticleComponents :type Array%20of%20FXSystem%20Component%20Object%20References) (AudioComponents :type Array%20of%20Audio%20Component%20Object%20References) (BurstCameraShakeInstance :type Camera%20Shake%20Base%20Object%20Reference) (BurstDecalInstance :type Decal%20Actor%20Object%20Reference))
  (Audio|Components|Audio|FadeOut (Variables|Default|GetFortLayeredAudio) 0.3 0.0 "Linear" :!class /Script/Engine.AudioComponent)
  (Utilities|Time|ClearandInvalidateTimerbyHandle 0 (Variables|Default|GetSoundIndicator_TH))
  (GameplayEventRouter|StopListeningtoEvent (Variables|Default|GetAsync_ListenToAmmoPickup)))

(event Custom|On%20Ammo%20Pickup
  (Variables|Default|SetLastAmmoPickupTime (Utilities|Time|GetGameTimeinSeconds (Variables|Default|GetCurrentWeapon))))

(fn ConstructionScript)
